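# Sound cue for the Paprika SMG's drop impact, mixing a landing sound with a sweetener layer.
# The landing layer reuses Paprika pistol drop samples at a fixed lowered pitch, with a quality switch choosing between random pools of four or two variants.
# The sweetener layer randomly picks one of two SMG samples through its own modulator.
# The cue overrides attenuation and caps concurrency, preventing new instances at the limit.
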SoundCue /PaprikaCoreWeapons/Items/Weapons/PaprikaSMG_DPS/Sounds/PaprikaSMG_DropImpact_Cue
root=Mixer_0
nodes [
 {id=Mixer_0, children [Modulator_0,Modulator_1], properties {inputVolume [1,1]}}
 {id=Modulator_0, children [QualityLevel_0], properties {pitchMax=0.9, pitchMin=0.9}}
 {id=Modulator_1, children [Random_2]}
 {id=QualityLevel_0, children [Random_1,Random_1,Random_0,Random_0]}
 {id=Random_0, children [WavePlayer_5,WavePlayer_6], properties {bShouldExcludeFromBranchCulling=true, weights [1,1]}}
 {id=Random_1, children [WavePlayer_3,WavePlayer_4,WavePlayer_5,WavePlayer_6], properties {weights [1,1,1,1]}}
 {id=Random_2, children [WavePlayer_1,WavePlayer_2], properties {weights [1,1]}}
 {id=WavePlayer_1, properties {soundWaveAssetPtr=&/PaprikaSMG_DropLand_Sweetner_02}}
 {id=WavePlayer_2, properties {soundWaveAssetPtr=&/PaprikaSMG_DropLand_Sweetner_03}}
 {id=WavePlayer_3, properties {soundWaveAssetPtr=/PaprikaCoreWeapons/Items/Weapons/PaprikaPistol/Sounds/Pistol_Paprika_AutoPistol_DropLand_01}}
 {id=WavePlayer_4, properties {soundWaveAssetPtr=/PaprikaCoreWeapons/Items/Weapons/PaprikaPistol/Sounds/Pistol_Paprika_AutoPistol_DropLand_02}}
 {id=WavePlayer_5, properties {soundWaveAssetPtr=/PaprikaCoreWeapons/Items/Weapons/PaprikaPistol/Sounds/Pistol_Paprika_AutoPistol_DropLand_03}}
 {id=WavePlayer_6, properties {soundWaveAssetPtr=/PaprikaCoreWeapons/Items/Weapons/PaprikaPistol/Sounds/Pistol_Paprika_AutoPistol_DropLand_04}}
]
properties {
 bOverrideAttenuation=true
 bOverrideConcurrency=true
 maxDistance=1000
 soundClassObject=&Fort_Loot_Collect
 volumeMultiplier=1
 attenuationOverrides {falloffDistance=1000, attenuationShapeExtents {x=0}}
 concurrencyOverrides {maxCount=3, resolutionRule=PreventNew}
}
