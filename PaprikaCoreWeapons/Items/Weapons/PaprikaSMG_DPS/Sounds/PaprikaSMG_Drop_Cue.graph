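# Sound cue for the drop sound of the Paprika SMG weapon, with weapon-drop attenuation, concurrency and a loot-collect sound class.
# A mixer layers two modulated branches: reused Paprika auto-pistol drop samples and a random SMG-specific sweetener.
# A quality level node switches the pistol layer between a four-sample random pool and a reduced two-sample pool.
# The cue plays pitched slightly up at reduced volume.
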SoundCue /PaprikaCoreWeapons/Items/Weapons/PaprikaSMG_DPS/Sounds/PaprikaSMG_Drop_Cue
root=Mixer_0
nodes [
 {id=Mixer_0, children [Modulator_1,Modulator_0], properties {inputVolume [1,1]}}
 {id=Modulator_0, children [Random_2]}
 {id=Modulator_1, children [QualityLevel_0]}
 {id=QualityLevel_0, children [Random_0,Random_0,Random_1,Random_1]}
 {id=Random_0, children [WavePlayer_1,WavePlayer_2,WavePlayer_3,WavePlayer_6], properties {weights [1,1,1,1]}}
 {id=Random_1, children [WavePlayer_3,WavePlayer_6], properties {bShouldExcludeFromBranchCulling=true, weights [1,1]}}
 {id=Random_2, children [WavePlayer_0,WavePlayer_4,WavePlayer_5], properties {weights [1,1,1]}}
 {id=WavePlayer_0, properties {soundWaveAssetPtr=&/PaprikaSMG_Drop_Sweetner_03}}
 {id=WavePlayer_1, properties {soundWaveAssetPtr=/PaprikaCoreWeapons/Items/Weapons/PaprikaPistol/Sounds/Pistol_Paprika_AutoPistol_Drop_04}}
 {id=WavePlayer_2, properties {soundWaveAssetPtr=/PaprikaCoreWeapons/Items/Weapons/PaprikaPistol/Sounds/Pistol_Paprika_AutoPistol_Drop_03}}
 {id=WavePlayer_3, properties {soundWaveAssetPtr=/PaprikaCoreWeapons/Items/Weapons/PaprikaPistol/Sounds/Pistol_Paprika_AutoPistol_Drop_02}}
 {id=WavePlayer_4, properties {soundWaveAssetPtr=&/PaprikaSMG_Drop_Sweetner_02}}
 {id=WavePlayer_5, properties {soundWaveAssetPtr=&/PaprikaSMG_Drop_Sweetner_01}}
 {id=WavePlayer_6, properties {soundWaveAssetPtr=/PaprikaCoreWeapons/Items/Weapons/PaprikaPistol/Sounds/Pistol_Paprika_AutoPistol_Drop_01}}
]
properties {
 attenuationSettings=&Fort_Weapon_Drop_Att
 maxDistance=3000
 pitchMultiplier=1.1
 soundClassObject=&Fort_Loot_Collect
 volumeMultiplier=0.6
 concurrencyOverrides {resolutionRule=PreventNew}
 concurrencySet [&Weapon_Drop_StopOldest]
}
