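# Sound cue for the equip foley of the PaprikaSMG DPS weapon.
# A Random node picks one of three equally weighted equip waves and feeds it through a Modulator left at default settings.
# It plays in the weapon foley sound class with the shared reload attenuation asset and a slightly lowered volume.
# The concurrency override and debug flag are both enabled.
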SoundCue /PaprikaCoreWeapons/Items/Weapons/PaprikaSMG_DPS/Sounds/PaprikaSMG_Equip_Cue
root=Modulator_0
nodes [
 {id=Modulator_0, children [Random_0]}
 {id=Random_0, children [WavePlayer_1,WavePlayer_2,WavePlayer_3], properties {weights [1,1,1]}}
 {id=WavePlayer_1, properties {soundWaveAssetPtr=&/PaprikaSMG_Equip_01}}
 {id=WavePlayer_2, properties {soundWaveAssetPtr=&/PaprikaSMG_Equip_02}}
 {id=WavePlayer_3, properties {soundWaveAssetPtr=&/PaprikaSMG_Equip_03}}
]
properties {
 attenuationSettings=&Fort_Weapon_Reload_Att
 bDebug=true
 bOverrideConcurrency=true
 maxDistance=2250
 soundClassObject=&Fort_WeaponFoley
 volumeMultiplier=0.902
}
