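# Sound cue for the Paprika SMG magazine insert during reload, played as weapon foley with reload attenuation.
# A mixer layers a randomized standard magazine insert, lightly modulated in pitch and volume, with a quieter random side-slap sound that starts after a short fixed delay.
# A quality-level node picks between the full insert variation set and a smaller subset, cutting variety at some quality settings.
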SoundCue /PaprikaCoreWeapons/Items/Weapons/PaprikaSMG_DPS/Sounds/PaprikaSMG_Magazine_Insert_Cue
root=Mixer_0
nodes [
 {id=Delay_0, children [Random_2], properties {delayMax=0.25, delayMin=0.25}}
 {id=Mixer_0, children [Modulator_0,Delay_0], properties {inputVolume [1,0.7]}}
 {id=Modulator_0, children [QualityLevel_0], properties {pitchMax=1.1, pitchMin=0.9, volumeMax=1.1, volumeMin=0.9}}
 {id=QualityLevel_0, children [Random_0,Random_0,Random_1,Random_1]}
 {id=Random_0, children [WavePlayer_0,WavePlayer_1,WavePlayer_2,WavePlayer_3,WavePlayer_4], properties {weights [1,1,1,1,1]}}
 {id=Random_1, children [WavePlayer_2,WavePlayer_3,WavePlayer_4], properties {weights [1,1,1]}}
 {id=Random_2, children [WavePlayer_5,WavePlayer_6,WavePlayer_7], properties {weights [1,1,1]}}
 {id=WavePlayer_0, properties {soundWaveAssetPtr=/Game/Sounds/Fort_Weapons/Magazines/Magazine_Standard_Insert_01}}
 {id=WavePlayer_1, properties {soundWaveAssetPtr=/Game/Sounds/Fort_Weapons/Magazines/Magazine_Standard_Insert_02}}
 {id=WavePlayer_2, properties {soundWaveAssetPtr=/Game/Sounds/Fort_Weapons/Magazines/Magazine_Standard_Insert_03}}
 {id=WavePlayer_3, properties {soundWaveAssetPtr=/Game/Sounds/Fort_Weapons/Magazines/Magazine_Standard_Insert_04}}
 {id=WavePlayer_4, properties {soundWaveAssetPtr=/Game/Sounds/Fort_Weapons/Magazines/Magazine_Standard_Insert_05}}
 {id=WavePlayer_5, properties {soundWaveAssetPtr=&/PaprikaSMG_SideSlap_01}}
 {id=WavePlayer_6, properties {soundWaveAssetPtr=&/PaprikaSMG_SideSlap_02}}
 {id=WavePlayer_7, properties {soundWaveAssetPtr=&/PaprikaSMG_SideSlap_03}}
]
properties {
 attenuationSettings=&Fort_Weapon_Reload_Att
 bOverrideConcurrency=true
 maxDistance=2250
 soundClassObject=&Fort_WeaponFoley
 volumeMultiplier=0.401
}
