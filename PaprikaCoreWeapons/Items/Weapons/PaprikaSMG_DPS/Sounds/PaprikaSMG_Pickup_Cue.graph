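# Pickup sound cue for the Paprika DPS SMG, using the loot-collect sound class, weapon pickup attenuation and a shared pickup concurrency set.
# A mixer layers reused Paprika pistol pickup waves, pitched down, with a louder random SMG-specific sweetener.
# A quality-level node switches the pistol layer between a four-wave random pool and a two-wave subset excluded from branch culling.
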SoundCue /PaprikaCoreWeapons/Items/Weapons/PaprikaSMG_DPS/Sounds/PaprikaSMG_Pickup_Cue
root=Mixer_0
nodes [
 {id=Mixer_0, children [Modulator_1,Modulator_0], properties {inputVolume [0.8,1.2]}}
 {id=Modulator_0, children [Random_2]}
 {id=Modulator_1, children [QualityLevel_0], properties {pitchMax=0.9, pitchMin=0.9}}
 {id=QualityLevel_0, children [Random_0,Random_0,Random_1,Random_1]}
 {id=Random_0, children [WavePlayer_10, WavePlayer_11, WavePlayer_12, WavePlayer_13], properties {weights [1,1,1,1]}}
 {id=Random_1, children [WavePlayer_12, WavePlayer_13], properties {bShouldExcludeFromBranchCulling=true, weights [1,1]}}
 {id=Random_2, children [WavePlayer_0,WavePlayer_1], properties {weights [1,1]}}
 {id=WavePlayer_0, properties {soundWaveAssetPtr=&/PaprikaSMG_Pickup_Sweetner_01}}
 {id=WavePlayer_1, properties {soundWaveAssetPtr=&/PaprikaSMG_Pickup_Sweetner_02}}
 {id=WavePlayer_10, properties {soundWaveAssetPtr=/PaprikaCoreWeapons/Items/Weapons/PaprikaPistol/Sounds/Pistol_Paprika_AutoPistol_PickUp_04}}
 {id=WavePlayer_11, properties {soundWaveAssetPtr=/PaprikaCoreWeapons/Items/Weapons/PaprikaPistol/Sounds/Pistol_Paprika_AutoPistol_PickUp_03}}
 {id=WavePlayer_12, properties {soundWaveAssetPtr=/PaprikaCoreWeapons/Items/Weapons/PaprikaPistol/Sounds/Pistol_Paprika_AutoPistol_PickUp_02}}
 {id=WavePlayer_13, properties {soundWaveAssetPtr=/PaprikaCoreWeapons/Items/Weapons/PaprikaPistol/Sounds/Pistol_Paprika_AutoPistol_PickUp_01}}
]
properties {
 attenuationSettings=&Fort_Weapon_Pickup_Att
 maxDistance=3500
 soundClassObject=&Fort_Loot_Collect
 volumeMultiplier=0.3
 concurrencyOverrides {maxCount=1, resolutionRule=PreventNew}
 concurrencySet [&PickupMaximum_Oldest]
}
