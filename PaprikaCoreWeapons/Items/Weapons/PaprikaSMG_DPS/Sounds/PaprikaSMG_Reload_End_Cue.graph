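# Sound cue for the end-of-reload foley of the PaprikaSMG weapon.
# A mixer layers two branches: one of three equally weighted reload-end variations after a fixed short delay, and an undelayed sweetener at slightly lower volume, each through a modulator.
# It uses the weapon foley sound class and shared reload attenuation, with concurrency override and the debug flag enabled.
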SoundCue /PaprikaCoreWeapons/Items/Weapons/PaprikaSMG_DPS/Sounds/PaprikaSMG_Reload_End_Cue
root=Mixer_0
nodes [
 {id=Delay_0, children [Modulator_0], properties {delayMax=0.1, delayMin=0.1}}
 {id=Mixer_0, children [Delay_0,Modulator_1], properties {inputVolume [1,0.8]}}
 {id=Modulator_0, children [Random_2]}
 {id=Modulator_1, children [WavePlayer_2]}
 {id=Random_2, children [WavePlayer_1,WavePlayer_8,WavePlayer_9], properties {weights [1,1,1]}}
 {id=WavePlayer_1, properties {soundWaveAssetPtr=&/PaprikaSMG_Reload_End_03}}
 {id=WavePlayer_2, properties {soundWaveAssetPtr=&/PaprikaSMG_ReloadEnd_Sweetner}}
 {id=WavePlayer_8, properties {soundWaveAssetPtr=&/PaprikaSMG_Reload_End_02}}
 {id=WavePlayer_9, properties {soundWaveAssetPtr=&/PaprikaSMG_Reload_End_01}}
]
properties {
 attenuationSettings=&Fort_Weapon_Reload_Att
 bDebug=true
 bOverrideConcurrency=true
 maxDistance=2250
 soundClassObject=&Fort_WeaponFoley
 volumeMultiplier=0.501
}
